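# Control Rig for the F_MED_Snowfall face accessory skeleton that procedurally offsets the right-side hair bones as the body poses.
# A blend weight adds a clamped remap of head rotation relative to spine_05 to a negative one from right clavicle-to-upper-arm rotation.
# That weight drives chained additive local offsets on dyn_hair_base_r and dyn_hair_parent_r, using Euler offsets from initial rotations.
# Left-hair steps have zero weight; several other nodes are disconnected and IsADS is unused.
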ControlRigBlueprint /BRCosmetics/Characters/Player/Female/Medium/Bodies/F_MED_Snowfall/Meshes/Parts/F_MED_Snowfall_FaceAcc_CtrlRig
graphs [
 RigVMFunctionLibrary {links [], nodes []}
 RigVMModel {nodes [
   RigUnit_BeginExecution = &RigUnit_BeginExecution()
   n1 = &RigUnit_GetRelativeTransformForItem(Child=(Type=Bone,Name="head"), Parent=(Type=Bone,Name="spine_05"), bChildInitial=false, bParentInitial=false)
   n2 = &RigVMFunction_MathQuaternionToEuler(
    RotationOrder=ZYX
    Value=&RigUnit_GetRelativeTransformForItem(Child=(Type=Bone,Name="clavicle_r"), Parent=(Type=Bone,Name="upperarm_r"), bChildInitial=false, bParentInitial=false).RelativeTransform.Rotation
   )
   n3 = &RigVMFunction_MathFloatAdd(
    A=&RigVMFunction_MathFloatRemap(
     SourceMaximum=0
     SourceMinimum=-70
     TargetMaximum=0
     TargetMinimum=1
     bClamp=true
     Value=&RigVMFunction_MathQuaternionToEuler(RotationOrder=ZYX, Value=@n1.RelativeTransform.Rotation).Result.Z
    ).Result
    B=&RigVMFunction_MathFloatRemap(SourceMaximum=-20, SourceMinimum=-50, TargetMaximum=-1, TargetMinimum=0, bClamp=true, Value=@n2.Result.Z).Result
   ) {template="Add::Execute(in A,in B,out Result)"}
   n4 = &RigVMFunction_MathQuaternionToEuler(
    RotationOrder=ZYX
    Value=/Script/ControlRig.RigUnit_GetInitialBoneTransform(Bone=dyn_hair_base_r, Space=LocalSpace) {is=struct RigVMUnitNode}.Transform.Rotation
   )
   /Script/ControlRig.RigUnit_ModifyBoneTransforms(
    BoneToModify=((Bone="dyn_hair_base_r",Transform=(Rotation=(X=0,Y=0,Z=0,W=1),Translation=(X=-4,Y=-2,Z=0),Scale3D=(X=1,Y=1,Z=1))))
    Mode=AdditiveLocal
    WeightMaximum=1
    WeightMinimum=0
    Weight=@n3.Result
    BoneToModify.0.Transform.Rotation=&RigVMFunction_MathQuaternionFromEuler(Euler=/Script/RigVM.RigVMFunction_MathVectorAdd(B=(X=50,Y=160,Z=-30), A=@n4.Result) {is=struct RigVMUnitNode, template="Add::Execute(in A,in B,out Result)"}.Result).Result
    ExecutePin=@RigUnit_BeginExecution.ExecutePin
   ) {is=struct RigVMUnitNode}
   > &RigUnit_ModifyBoneTransforms(
    BoneToModify=((Bone="dyn_hair_parent_r",Transform=(Rotation=(X=0,Y=0,Z=0,W=1),Translation=(X=-4,Y=0,Z=0),Scale3D=(X=1,Y=1,Z=1))))
    Mode=AdditiveLocal
    WeightMaximum=1
    WeightMinimum=0
    Weight=@n3.Result
   )
   > &RigUnit_ModifyTransforms(
    ItemToModify=((Item=(Type=Bone,Name="dyn_hair_01_r"),Transform=(Rotation=(X=0,Y=0,Z=0,W=1),Translation=(X=0,Y=0,Z=0),Scale3D=(X=1,Y=1,Z=1))))
    Weight=1
   )
   n5 = &RigVMFunction_MathQuaternionToEuler(RotationOrder=YXZ, Value=&RigUnit_GetInitialBoneTransform(Bone=dyn_hair_base_r, Space=LocalSpace).Transform.Rotation)
   > &RigUnit_ModifyBoneTransforms(
    BoneToModify=((Bone="dyn_hair_base_r",Transform=(Rotation=(X=0,Y=0,Z=0,W=1),Translation=(X=0,Y=0,Z=0),Scale3D=(X=1,Y=1,Z=1))))
    Mode=AdditiveLocal
    WeightMaximum=1
    WeightMinimum=0
    Weight=@n3.Result
    BoneToModify.0.Transform.Rotation=&RigVMFunction_MathQuaternionFromEuler(
     RotationOrder=XZY
     Euler=&RigVMFunction_MathVectorAdd(B=(X=-20,Y=180,Z=0), A=@n5.Result) {template="Add::Execute(in A,in B,out Result)"}.Result
    ).Result
   )
   n6 = &RigVMFunction_MathQuaternionToEuler(RotationOrder=ZYX, Value=&RigUnit_GetInitialBoneTransform(Bone=dyn_hair_base_l, Space=LocalSpace).Transform.Rotation)
   > &RigUnit_ModifyTransforms(
    ItemToModify=((Item=(Type=Bone,Name="dyn_hair_base_l"),Transform=(Rotation=(X=0,Y=0,Z=0,W=1),Translation=(X=0,Y=0,Z=0),Scale3D=(X=1,Y=1,Z=1))))
    Weight=0
    ItemToModify.0.Transform.Rotation=&RigVMFunction_MathQuaternionFromEuler(Euler=&RigVMFunction_MathVectorAdd(B=(X=90,Y=-180,Z=0), A=@n6.Result) {template="Add::Execute(in A,in B,out Result)"}.Result).Result
   )
   n7 = &RigVMFunction_MathQuaternionToEuler(RotationOrder=ZYX, Value=&RigUnit_GetInitialBoneTransform(Bone=dyn_hair_base_l, Space=LocalSpace).Transform.Rotation)
   > &RigUnit_ModifyBoneTransforms(
    BoneToModify=((Bone="dyn_hair_base_l",Transform=(Rotation=(X=0,Y=0,Z=0,W=1),Translation=(X=2,Y=-1,Z=-5),Scale3D=(X=1,Y=1,Z=1))))
    Mode=AdditiveLocal
    Weight=0
    WeightMaximum=1
    WeightMinimum=0
    BoneToModify.0.Transform.Rotation=&RigVMFunction_MathQuaternionFromEuler(Euler=&RigVMFunction_MathVectorAdd(B=(X=70,Y=-140,Z=-30), A=@n7.Result) {template="Add::Execute(in A,in B,out Result)"}.Result).Result
   )
   &RigVMFunction_MathFloatAdd(A=@n2.Result.Z) {template="Add::Execute(in A,in B,out Result)"}
   n8 = &RigUnit_GetRelativeTransformForItem(Child=(Type=Bone,Name="head"), Parent=(Type=Bone,Name="spine_05"), bChildInitial=false, bParentInitial=false)
   &RigVMFunction_MathFloatAdd(A=&RigVMFunction_MathFloatRemap(
     SourceMaximum=60
     SourceMinimum=20
     TargetMinimum=0
     bClamp=true
     Value=&RigVMFunction_MathQuaternionToEuler(RotationOrder=ZYX, Value=@n8.RelativeTransform.Rotation).Result.Z
    ).Result) {template="Add::Execute(in A,in B,out Result)"}
   &RigVMFunction_MathFloatRemap(
    SourceMaximum=-10
    SourceMinimum=-50
    TargetMaximum=0
    TargetMinimum=1
    bClamp=true
    Value=&RigVMFunction_MathQuaternionToEuler(
     RotationOrder=ZYX
     Value=&RigUnit_GetRelativeTransformForItem(Child=(Type=Bone,Name="head"), Parent=(Type=Bone,Name="spine_05"), bChildInitial=false, bParentInitial=false).RelativeTransform.Rotation
    ).Result.X
   )
   &RigUnit_ModifyBoneTransforms(
    BoneToModify=((Bone="dyn_hair_12_r",Transform=(Rotation=(X=0,Y=0,Z=0,W=1),Translation=(X=-5,Y=-2,Z=0),Scale3D=(X=1,Y=1,Z=1))))
    ExecutePin=()
    Mode=AdditiveLocal
    WeightMaximum=1
    WeightMinimum=0
    Weight=@n3.Result
   )
   n9 = &RigVMFunction_MathQuaternionToEuler(RotationOrder=YXZ, Value=&RigUnit_GetInitialBoneTransform(Bone=dyn_hair_base_r, Space=LocalSpace).Transform.Rotation)
   &RigUnit_ModifyBoneTransforms(
    BoneToModify=((Bone="dyn_hair_11_r",Transform=(Rotation=(X=0,Y=0,Z=0,W=1),Translation=(X=0,Y=0,Z=0),Scale3D=(X=1,Y=1,Z=1))))
    ExecutePin=()
    Mode=AdditiveLocal
    Weight=0
    WeightMaximum=1
    WeightMinimum=0
    BoneToModify.0.Transform.Rotation=&RigVMFunction_MathQuaternionFromEuler(
     RotationOrder=XZY
     Euler=&RigVMFunction_MathVectorAdd(B=(X=0,Y=150,Z=0), A=@n9.Result) {template="Add::Execute(in A,in B,out Result)"}.Result
    ).Result
   )
   RigVMVariableNode(ExecuteContext=(), Value=@n3.Result) {variable=tempwatch}
  ]}
]
hierarchy [
 root {loc=<0 0 0>, rot=<0 0 0 1>}
 pelvis {parent=root, loc=<0 -0.5301 98.0571>, rot=<0.0127 0.707 -0.0127 -0.707>}
 spine_01 {parent=pelvis, loc=<6.1208 -0.9328 0>, rot=<0 0 0.034 -0.9994>}
 spine_02 {parent=spine_01, loc=<6.5996 -0.2095 0>, rot=<0 0 0.0238 -0.9997>}
 spine_03 {parent=spine_02, loc=<6.4914 -0.1237 0>, rot=<0 0 0.0149 -0.9999>}
 spine_04 {parent=spine_03, loc=<6.9598 0.0794 0>, rot=<0 0 -0.0132 -0.9999>}
 spine_05 {parent=spine_04, loc=<6.9987 0 0>, rot=<0 0 -0.0015 -1>}
 clavicle_l {parent=spine_05, loc=<5.4831 -0.1053 -3.2946>, rot=<0.0465 -0.7289 -0.009 -0.6829>}
 upperarm_l {parent=clavicle_l, loc=<9.8658 -1.2136 -0.4104>, rot=<0.0164 -0.3245 -0.0269 -0.9454>}
 clavicle_r {parent=spine_05, loc=<5.4831 -0.1053 3.2946>, rot=<0.7289 0.0465 -0.6829 0.009>}
 upperarm_r {parent=clavicle_r, loc=<-9.8658 1.2136 0.4104>, rot=<0.0164 -0.3245 -0.0269 -0.9454>}
 neck_01 {parent=spine_05, loc=<11.3138 0.1529 0>, rot=<0 0 -0.2003 -0.9797>}
 neck_02 {parent=neck_01, loc=<4.5627 0 0>, rot=<0 0 0.0504 -0.9987>}
 head {parent=neck_02, loc=<5.3914 0 0>, rot=<0 0 0.115 -0.9934>}
 face_acc_offset {parent=head, loc=<2.7283 0.0226 0>, rot=<0 0 0 1>}
 dyn_beanie_base {parent=face_acc_offset, loc=<8.7114 -6.2346 0.2774>, rot=<-0.6567 0.2622 0.2622 -0.6567>}
 dyn_beanie {parent=dyn_beanie_base, loc=<0 0 0>, rot=<0 0 0 1>}
 dyn_strand_base_l {parent=face_acc_offset, loc=<-3.008 5.3036 -7.1916>, rot=<0.0963 -0.276 0.9055 0.3076>}
 dyn_strand_parent_l {parent=dyn_strand_base_l, loc=<0 0 0>, rot=<0 0 0 1>}
 dyn_strand_01_l {parent=dyn_strand_parent_l, loc=<2.377 -0.0287 -0.1397>, rot=<0 0 0.0698 -0.9976>}
 dyn_strand_02_l {parent=dyn_strand_01_l, loc=<3.6999 -0.1014 -0.5164>, rot=<-0.0186 -0.2121 -0.1654 -0.963>}
 dyn_strand_01 {parent=dyn_strand_parent_l, loc=<2.377 -0.0287 -0.1397>, rot=<0 0 0.0698 -0.9976>}
 dyn_strand_02 {parent=dyn_strand_01, loc=<3.6999 -0.1014 -0.5164>, rot=<-0.0186 -0.2121 -0.1654 -0.963>}
 dyn_hair_base_l {parent=face_acc_offset, loc=<-3.9265 -1.6103 -9.0112>, rot=<-0.2863 -0.6073 0.6983 0.2482>}
 dyn_hair_parent_l {parent=dyn_hair_base_l, loc=<0 0 0>, rot=<0 0 0 1>}
 dyn_hair_01_l {parent=dyn_hair_parent_l, loc=<6.8743 -0.0716 0.0878>, rot=<0.141 0 0.141 -0.9799>}
 dyn_hair_02_l {parent=dyn_hair_01_l, loc=<3.8463 -0.118 -0.5093>, rot=<-0.1209 -0.0629 0.1922 -0.9718>}
 dyn_hair_03_l {parent=dyn_hair_02_l, loc=<4.9519 0.1721 0.1947>, rot=<-0.0627 0.1784 -0.1729 -0.9666>}
 dyn_hair_04_l {parent=dyn_hair_03_l, loc=<3.9857 0.1821 2.1647>, rot=<0.0288 0.0133 -0.2078 -0.9777>}
 dyn_strand_base_r {parent=face_acc_offset, loc=<0.7763 6.8702 6.0744>, rot=<0.0228 -0.9429 0.2536 0.2146>}
 dyn_strand_parent_r {parent=dyn_strand_base_r, loc=<0 0 0>, rot=<0 0 0 1>}
 dyn_strand_01_r {parent=dyn_strand_parent_r, loc=<3.9327 -0.3634 -0.7311>, rot=<0.1396 -0.1991 0.1971 -0.9498>}
 dyn_strand_02_r {parent=dyn_strand_01_r, loc=<2.6017 -1.4143 -0.1316>, rot=<-0.0092 0.0565 0.569 -0.8204>}
 dyn_hair_base_r {parent=face_acc_offset, loc=<-1.8256 0.8526 7.9714>, rot=<0.004 -0.9563 0.0012 0.2924>}
 dyn_hair_parent_r {parent=dyn_hair_base_r, loc=<0 0 0>, rot=<0 0 0 1>}
 dyn_hair_01_r {parent=dyn_hair_parent_r, loc=<4.6915 0.6614 0.6908>, rot=<-0.0012 0.0282 -0.0822 -0.9962>}
 dyn_hair_02_r {parent=dyn_hair_01_r, loc=<3.4209 0.9236 0.8028>, rot=<0.0365 0.1368 -0.0629 -0.9879>}
 dyn_hair_03_r {parent=dyn_hair_02_r, loc=<3.384 0.3746 0.2843>, rot=<-0.048 0.1174 -0.039 -0.9912>}
 dyn_hair_04_r {parent=dyn_hair_03_r, loc=<2.6777 -0.1064 -0.19>, rot=<0.2518 0.0063 0.0535 -0.9663>}
 dyn_hair_05_r {parent=dyn_hair_04_r, loc=<2.9855 0.2021 1.0572>, rot=<0.0799 0.3864 -0.0487 -0.9176>}
 dyn_hair_11_r {parent=dyn_hair_parent_r, loc=<3.6154 -2.817 0.8419>, rot=<0.9867 0.0287 0.1268 0.0972>}
 dyn_hair_12_r {parent=dyn_hair_11_r, loc=<3.9105 0 0>, rot=<0.0303 -0.0881 0.2778 -0.9561>}
 dyn_collar_l {parent=spine_05, loc=<7.1842 5.7605 -6.9864>, rot=<-0.0696 -0.3657 -0.3462 -0.8611>}
 dyn_collar_r {parent=spine_05, loc=<7.0911 6.7009 7.2245>, rot=<0.8651 0.289 0.4044 -0.0672>}
]
variables [
 IsADS {type=bool, default=False}
 tempwatch {type=double, default=-1}
]
