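# Post-processes the face-accessory skeleton of the RedOasisApricot Icon female cosmetic, mainly driving hair and strand bones.
# Poses copied from parent and body meshes are layered, run through the Snowfall Control Rig and cached.
# Two RigidBody simulations, switched by front-end state, feed hair constraints and additive bone offsets, one applied only for one-handed melee.
# Event graph derives head-rotation alphas and the melee flag; head-driven constraints are disconnected from the output.
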AnimBlueprint /BRCosmetics/Characters/Player/Female/Medium/Bodies/F_MED_RedOasisApricot/Meshes/Icon/F_MED_RedOasisApricotIcon_FaceAcc_AnimBP
target_skeleton=/BRCosmetics/Characters/Player/Female/Medium/Bodies/F_MED_Snowfall/Meshes/Parts/F_MED_Snowfall_FaceAcc_Skeleton
graphs [
 AnimGraph {nodes [
   n1 = Get HeadLeftAlpha() -> (HeadLeftAlpha:double)
   "Constraint - dyn_hair_base_l (PPPP)"!(
    ComponentPose
    Alpha=@n1
    ConstraintWeights_0=0.4269
    ConstraintWeights_1:float=0.0237
    ConstraintWeights_2:float=0.0158
    ConstraintWeights_3:float=0.5336
   ) -> (Pose) {
    alphaBoolBlend {blendOption=Linear}
    boneToModify {boneName=dyn_hair_base_l}
    constraintSetup [{targetBone {boneName=spine_05}}, {targetBone {boneName=clavicle_l}}, {targetBone {boneName=neck_01}}, {targetBone {boneName=dyn_turtle_l}}]
    constraintWeights [1,1,1,1]
   }
   n2 = "Constraint - dyn_hair_11_r (PP)"!(ComponentPose, Alpha=@n1, ConstraintWeights_0=0.2, ConstraintWeights_1:float=0.8) -> (Pose) {
    alphaBoolBlend {blendOption=Linear}
    boneToModify {boneName=dyn_hair_11_r}
    constraintSetup [{targetBone {boneName=spine_05}}, {targetBone {boneName=head}}]
    constraintWeights [1,1]
   }
   "Constraint - dyn_hair_base_r (PP)"!(ComponentPose=@n2, Alpha=@n1, ConstraintWeights_0=0.2, ConstraintWeights_1:float=0.8) -> (Pose) {
    alphaBoolBlend {blendOption=Linear}
    boneToModify {boneName=dyn_hair_base_r}
    constraintSetup [{targetBone {boneName=dyn_hair_11_r}}, {targetBone {boneName=head}}]
    constraintWeights [1,1]
   }
   "Constraint - dyn_hair_base_l (PPT)"!(ComponentPose, Alpha=@n1, ConstraintWeights_0=0.4, ConstraintWeights_1:float=0.6, ConstraintWeights_2:float=1) -> (Pose) {
    alphaBoolBlend {blendOption=Linear}
    boneToModify {boneName=dyn_hair_base_l}
    constraintSetup [{targetBone {boneName=spine_05}}, {targetBone {boneName=clavicle_l}}, {transformType=Translation, targetBone {boneName=None}}]
    constraintWeights [1,1,1]
   }
   "Transform (Modify) Bone - Bone: dyn_hair_11_r"(Alpha=@n1, Translation="0,0,5", Rotation="0,0,0", Scale="1,1,1") {translationMode=BMM_Additive, alphaBoolBlend {blendOption=Linear}, boneToModify {boneName=dyn_hair_11_r}, translation {z=5}}
   n3 = RigidBody(ComponentPose=Local To Component(LocalPose=Use cached pose 'Before Rban'())) {
    bClampLinearTranslationLimitToRefPose=true
    bOverrideWorldGravity=true
    overridePhysicsAsset=&/F_MED_RedOasisApricotIcon_FaceAcc_Physics
    simulationSpace=BaseBoneSpace
    alphaBoolBlend {blendOption=Linear}
    baseBoneRef {boneName=spine_05}
    componentAppliedLinearAccClamp=<6000 6000 6000>
    componentLinearAccScale=<0.15 0.15 0.25>
    componentLinearVelScale=<0.15 0.15 0.25>
    overrideWorldGravity {z=-980}
   }
   n4 = Component To Local(ComponentPose=@n3)
   n5 = RigidBody(ComponentPose=Local To Component(LocalPose=Use cached pose 'Before Rban'())) {
    bClampLinearTranslationLimitToRefPose=true
    bOverrideWorldGravity=true
    overridePhysicsAsset=&F_MED_RedOasisApricotIcon_FaceAcc_Physics
    simulationSpace=BaseBoneSpace
    alphaBoolBlend {blendOption=Linear}
    baseBoneRef {boneName=spine_05}
    componentAppliedLinearAccClamp=<6000 6000 6000>
    componentLinearAccScale=<0.15 0.15 0.25>
    componentLinearVelScale=<0.15 0.15 0.25>
    overrideWorldGravity {z=-980}
   }
   n6 = Component To Local(ComponentPose=@n5)
   n7 = Get bIsInFrontEnd()
   n8 = "Constraint - dyn_hair_02_r (P)"(
    ComponentPose=Local To Component(LocalPose=Blend Poses by bool(BlendPose_0=@n4, BlendPose_1=@n6, bActiveValue=@n7) {blendType=Linear})
    Alpha=0.3
    ConstraintWeights_0=1
   ) {
    alphaBoolBlend {blendOption=Linear}
    boneToModify {boneName=dyn_hair_02_r}
    constraintSetup [{targetBone {boneName=dyn_hair_11_r}}]
    constraintWeights [1]
   }
   n9 = "Constraint - dyn_strand_base_l (R)"(ComponentPose=@n8, Alpha=1, ConstraintWeights_0=1) {
    alphaBoolBlend {blendOption=Linear}
    boneToModify {boneName=dyn_strand_base_l}
    constraintSetup [{transformType=Rotation, targetBone {boneName=dyn_hair_base_l}}]
    constraintWeights [1]
   }
   n10 = "Constraint - dyn_strand_01_r (R)"(ComponentPose=@n9, Alpha=0.7, ConstraintWeights_0=1) {
    alphaBoolBlend {blendOption=Linear}
    boneToModify {boneName=dyn_strand_01_r}
    constraintSetup [{transformType=Rotation, targetBone {boneName=dyn_hair_01_r}}]
    constraintWeights [1]
   }
   n11 = "Constraint - dyn_strand_02_r (R)"(ComponentPose=@n10, Alpha=0.2, ConstraintWeights_0=1) {
    alphaBoolBlend {blendOption=Linear}
    boneToModify {boneName=dyn_strand_02_r}
    constraintSetup [{transformType=Rotation, targetBone {boneName=dyn_hair_01_r}}]
    constraintWeights [1]
   }
   n12 = "Constraint - dyn_strand_01_l (R)"(ComponentPose=@n11, Alpha=0.8, ConstraintWeights_0=1) {
    alphaBoolBlend {blendOption=Linear}
    boneToModify {boneName=dyn_strand_01_l}
    constraintSetup [{transformType=Rotation, targetBone {boneName=dyn_hair_01_l}}]
    constraintWeights [1]
   }
   n13 = "Constraint - dyn_strand_02_l (R)"(ComponentPose=@n12, Alpha=0.5, ConstraintWeights_0=1) {
    alphaBoolBlend {blendOption=Linear}
    boneToModify {boneName=dyn_strand_02_l}
    constraintSetup [{transformType=Rotation, targetBone {boneName=dyn_hair_02_l}}]
    constraintWeights [1]
   }
   n14 = "Transform (Modify) Bone - Bone: dyn_hair_11_r"(
    ComponentPose=@n13
    Alpha=Get IsOneHandMeleeAnim() -> (IsOneHandMeleeAnim:double)
    Translation="-1.5,0,0"
    Rotation="10.000,-0.000,-0.001"
    Scale="1,1,1"
   ) {
    rotationMode=BMM_Additive
    rotationSpace=BCS_BoneSpace
    translationMode=BMM_Additive
    translationSpace=BCS_BoneSpace
    alphaBoolBlend {blendOption=Linear}
    boneToModify {boneName=dyn_hair_11_r}
    rotation=<rot 10 -0.0001 -0.0009>
    translation {x=-1.5}
   }
   n15 = "Transform (Modify) Bone - Bone: dyn_hair_12_r"(ComponentPose=@n14, Alpha=1, Translation="0,8,0", Rotation="0,0,0", Scale="1,1,1") {
    rotationMode=BMM_Additive
    rotationSpace=BCS_BoneSpace
    translationMode=BMM_Additive
    translationSpace=BCS_BoneSpace
    alphaBoolBlend {blendOption=Linear}
    boneToModify {boneName=dyn_hair_12_r}
    translation {y=8}
   }
   n16 = "Transform (Modify) Bone - Bone: dyn_hair_04_r"(ComponentPose=@n15, Alpha=1, Translation="0,1,0", Rotation="-20,0,0", Scale="1,1,1") {
    rotationMode=BMM_Additive
    translationMode=BMM_Additive
    alphaBoolBlend {blendOption=Linear}
    boneToModify {boneName=dyn_hair_04_r}
    rotation {pitch=-20}
    translation {y=1}
   }
   n17 = "Transform (Modify) Bone - Bone: dyn_hair_03_r"(ComponentPose=@n16, Alpha=1, Translation="-2,0,0", Rotation="0,0,0", Scale="1,1,1") {
    rotationMode=BMM_Additive
    translationMode=BMM_Additive
    alphaBoolBlend {blendOption=Linear}
    boneToModify {boneName=dyn_hair_03_r}
    translation {x=-2}
   }
   n18 = "Transform (Modify) Bone - Bone: dyn_hair_02_r"(ComponentPose=@n17, Alpha=1, Translation="-3,1,0", Rotation="20,0,0", Scale="1,1,1") {
    rotationMode=BMM_Additive
    translationMode=BMM_Additive
    alphaBoolBlend {blendOption=Linear}
    boneToModify {boneName=dyn_hair_02_r}
    rotation {pitch=20}
    translation=<-3 1>
   }
   n19 = "Transform (Modify) Bone - Bone: dyn_hair_01_r"(ComponentPose=@n18, Alpha=1, Translation="0,1,0", Rotation="-1,0,-3", Scale="1,1,1") {
    rotationMode=BMM_Additive
    translationMode=BMM_Additive
    alphaBoolBlend {blendOption=Linear}
    boneToModify {boneName=dyn_hair_01_r}
    rotation {pitch=-1, roll=-3}
    translation {y=1}
   }
   n20 = "Transform (Modify) Bone - Bone: dyn_hair_12_r"(ComponentPose=@n19, Alpha=1, Translation="-1.5,0,0", Rotation="0,0,0", Scale="1,1,1") {
    rotationMode=BMM_Additive
    translationMode=BMM_Additive
    alphaBoolBlend {blendOption=Linear}
    boneToModify {boneName=dyn_hair_12_r}
    translation {x=-1.5}
   }
   n21 = "Transform (Modify) Bone - Bone: dyn_hair_01_l"(ComponentPose=@n20, Alpha=1, Translation="1,0,0", Rotation="0,0,0", Scale="1,1,1") {
    rotationMode=BMM_Additive
    translationMode=BMM_Additive
    alphaBoolBlend {blendOption=Linear}
    boneToModify {boneName=dyn_hair_01_l}
    translation {x=1}
   }
   n22 = "Transform (Modify) Bone - Bone: dyn_hair_02_l"(ComponentPose=@n21, Alpha=1, Translation="0,1,1.5", Rotation="-20,0,0", Scale="1,1,1") {
    rotationMode=BMM_Additive
    translationMode=BMM_Additive
    alphaBoolBlend {blendOption=Linear}
    boneToModify {boneName=dyn_hair_02_l}
    rotation {pitch=-20}
    translation {y=1, z=1.5}
   }
   n23 = "Transform (Modify) Bone - Bone: dyn_hair_03_l"(ComponentPose=@n22, Alpha=1, Translation="0,0.5,0", Rotation="0,0,0", Scale="1,1,1") {
    rotationMode=BMM_Additive
    translationMode=BMM_Additive
    alphaBoolBlend {blendOption=Linear}
    boneToModify {boneName=dyn_hair_03_l}
    translation {y=0.5}
   }
   n24 = "Transform (Modify) Bone - Bone: dyn_hair_04_l"(ComponentPose=@n23, Alpha=1, Translation="0,1,0.5", Rotation="0,0,0", Scale="1,1,1") {
    rotationMode=BMM_Additive
    translationMode=BMM_Additive
    alphaBoolBlend {blendOption=Linear}
    boneToModify {boneName=dyn_hair_04_l}
    translation {y=1, z=0.5}
   }
   Output Pose(Result=Component To Local(ComponentPose=@n24)) {layerGroup=None}
   n25 = Layered blend per bone(
    BasePose=Copy Pose From Mesh() {bUseAttachedParent=true}
    BlendPoses_0=Copy Pose From Mesh(SourceMeshComponent=Get Body Mesh() -> (Body Mesh:Skeletal Mesh Component Object Reference))
   ) {layerSetup [{branchFilters [
       {boneName=dyn_collar_l}
       {boneName=dyn_collar_r}
       {blendDepth=1, boneName=clavical_l}
       {blendDepth=1, boneName=clavical_r}
       {boneName=upperarm_l}
       {boneName=upperarm_r}
       {boneName=dyn_turtler_r}
      ]}]}
   n26 = Control Rig(Source=@n25) {
    alphaBoolBlend {blendOption=Linear}
    controlRigAssetReference {blueprintRigClass=/BRCosmetics/Characters/Player/Female/Medium/Bodies/F_MED_Snowfall/Meshes/Parts/F_MED_Snowfall_FaceAcc_CtrlRig.F_MED_Snowfall_FaceAcc_CtrlRig_C}
   }
   Save cached pose 'Before Rban'(Pose=@n26) {cacheName=Before Rban}
   Get IsADSAnim() -> (IsADSAnim:bool)
  ]}
 EventGraph {nodes [
   Event BlueprintUpdateAnimation()
   > n1 = Sequence!(execute) -> (then_0, then_1, then_2:Exec)
   > n2 = Cast To PlayerPawn_Athena(Object=TryGetPawnOwner())
   > Set Body Mesh(Body Mesh:Skeletal Mesh Component Object Reference=GetSkeletalMeshForPartType(self=@n2)) -> (Output_Get:Skeletal Mesh Component Object Reference)
   n3 = GetOwnerSkeletalMeshComponent()
   n4 = BreakTransform(InTransform=GetDeltaTransformFromRefPose(self=@n3, BoneName=head, BaseName=spine_05))
   > n5 = Set HeadLeftAlpha(HeadLeftAlpha:double=MapRangeClamped(Value=BreakRotator(InRot=@n4.Rotation), InRangeA=10, InRangeB=60)) -> (Output_Get:double)
   n6 = GetDeltaTransformFromRefPose(self=@n3, BoneName=head, BaseName=neck_02)
   n7 = MapRangeClamped(Value=BreakRotator(InRot=BreakTransform(InTransform=@n6).Rotation).Yaw, InRangeA=5, InRangeB=-12)
   Set HeadBack(execute=[@n1.then_2, @n5.then], HeadBack:double=@n7) -> (Output_Get:double)
   n8 = Cast To FortPlayerAnimInstance(execute=@n1.then_1, Object=GetAnimInstance(self=GetOwnerSkeletalMeshComponent()))
   > Set IsOneHandMeleeAnim(IsOneHandMeleeAnim:double=0) -> (Output_Get:double)
   > n9 = Switch on EFortWeaponCoreAnimation!(execute, Selection:EFortWeaponCoreAnimation Enum=Get WeaponCoreAnim(self=@n8)) -> (
    Melee:Exec
    Pistol:Exec
    Shotgun:Exec
    PaperBlueprint:Exec
    Rifle:Exec
    MeleeOneHand:Exec
    MachinePistol:Exec
    RocketLauncher:Exec
    GrenadeLauncher:Exec
    GoingCommando:Exec
    AssaultRifle:Exec
    TacticalShotgun:Exec
    SniperRifle:Exec
    TrapPlacement:Exec
    ShoulderLauncher:Exec
    AbilityDecoTool:Exec
    Crossbow:Exec
    C4:Exec
    RemoteControl:Exec
    DualWield:Exec
    AR_BullPup:Exec
    AR_ForwardGrip:Exec
    MedPackPaddles:Exec
    SMG_P90:Exec
    AR_DrumGun:Exec
    Consumable_Small:Exec
    Consumable_Large:Exec
    Balloon:Exec
    MountedTurret:Exec
    CreativeTool:Exec
    ExplosiveBow:Exec
    AshtonIndigo:Exec
    AshtonChicago:Exec
    MeleeDualWield:Exec
    Unarmed:Exec
   )
   Branch(execute=[@n9.AssaultRifle, @n9.Crossbow, @n9.GrenadeLauncher, @n9.Rifle, @n9.RocketLauncher, @n9.Shotgun, @n9.SniperRifle, @n9.TacticalShotgun])
   Branch(execute=[@n9.MachinePistol, @n9.Pistol])
   Branch(execute=@n9.MeleeOneHand)
   > Set IsOneHandMeleeAnim(IsOneHandMeleeAnim=1)
   n10 = GetDeltaTransformFromRefPose(self=GetOwnerSkeletalMeshComponent(), BoneName=head, BaseName=spine_05)
   MapRangeClamped(Value=BreakRotator(InRot=BreakTransform(InTransform=@n10).Rotation).Yaw, InRangeA=-20, InRangeB=0)
   GetOwnerSkeletalMeshComponent()
   float * float(A=@n7, B=1)
  ]}
]
variables [
 Body Mesh {type=SkeletalMeshComponent}
 HeadBack {type=number}
 HeadLeftAlpha {type=number}
 IsADSAnim {type=boolean}
 IsOneHandMeleeAnim {type=number}
]
